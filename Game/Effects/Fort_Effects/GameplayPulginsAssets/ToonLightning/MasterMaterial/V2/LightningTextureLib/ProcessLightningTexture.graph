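# Editor utility that batch-configures selected Texture2D assets, apparently as lightning FX textures for the ToonLightning material library.
# Its ProcessLightningFXTexture function loops over the selected textures and overwrites color space, compression and texture group settings for linear effects data.
# It also sets addressing to clamp on X and wrap on Y, then force-saves each asset.
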EditorUtilityBlueprint /Game/Effects/Fort_Effects/GameplayPulginsAssets/ToonLightning/MasterMaterial/V2/LightningTextureLib/ProcessLightningTexture
graphs [
 EventGraph {links [], nodes []}
 ProcessLightningFXTexture {nodes [
   ProcessLightningFXTexture() {is=title K2Node_FunctionEntry}
   > n1 = GetSelectedAssetsOfClass(AssetClass=/Script/Engine.Texture2D)
   > n2 = For Each Loop(Array:Array of Object References=@n1) -> (Array Element:Object Reference)
   > n3 = Cast To Texture2D(Object=@n2)
   > Set SRGB(SRGB=false, self=@n3)
   > Set CompressionSettings(CompressionSettings=TC_VectorDisplacementmap, self=@n3)
   > Set CompressionNoAlpha!(execute, CompressionNoAlpha:bool=true, self:Texture Object Reference=@n3) -> (then, Output_Get:bool)
   > Set LODGroup!(execute, LODGroup:TextureGroup Enum=TEXTUREGROUP_Effects, self:Texture Object Reference=@n3) -> (then, Output_Get:TextureGroup Enum)
   > Set AddressX!(execute, AddressX:TextureAddress Enum=TA_Clamp, self:Texture 2D Object Reference=@n3) -> (then, Output_Get:TextureAddress Enum=TA_Wrap)
   > Set AddressY!(execute, AddressY:TextureAddress Enum=TA_Wrap, self:Texture 2D Object Reference=@n3) -> (then, Output_Get:TextureAddress Enum=TA_Wrap)
   > SaveLoadedAsset(AssetToSave=@n3, bOnlyIfIsDirty=false)
  ]}
]
variables []
